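# M_FX_Toon_Unlit_GradientMapped is a masked unlit toon material for Niagara sprites, ribbons and meshes.
# A panned, distorted mask, optionally shaped by fresnel and alpha, is posterized into steps, remapped through a gradient texture, tinted by particle color and optionally hue-shifted.
# Fresnel and mask-threshold outlines blend into color and alpha.
# Dynamic parameters drive UV offsets, mask blends and alpha erosion, with optional view-angle fade.
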Material /BRCosmetics/Effects/Materials/Toon/M_FX_Toon_Unlit_GradientMapped
graph {
 properties {
  bAutomaticallySetUsageInEditor=false
  bCastRayTracedShadows=false
  bHasPixelAnimation=true
  bUsedWithNiagaraMeshParticles=true
  bUsedWithNiagaraRibbons=true
  bUsedWithNiagaraSprites=true
  blendMode=BLEND_Masked
  shadingModel=MSM_Unlit
  usedShadingModels=Unlit
 }
 nodes [
  n1 = DynamicParameter() {defaultValue=<rgba 1 1 0 1>, paramNames ["Outline Lerp (Enable in Mat first)", Mask Lerp, Alpha Erosion, Alpha Mask Lerp]}
  LinearInterpolate(
   A=@n1.Alpha Erosion
   B=Multiply(A=@n1.Alpha Erosion, B=NamedRerouteUsage())
   Alpha=ScalarParameter() {group=Alpha, parameterName=Use Angle Fade, sortPriority=36}
  )
  n2 = ScalarParameter() {group=Color, parameterName=Use Hue Shift, sortPriority=36}
  n3 = NamedRerouteUsage()
  n4 = NamedRerouteUsage()
  n5 = Saturate(None=ScalarParameter() {group=UVs, parameterName=Mask UV Select, sortPriority=30})
  n6 = ChannelMaskParameter(None=TextureSampleParameter2D(UVs=LinearInterpolate(A=@n3, B=@n4, Alpha=@n5)) {group=Mask, parameterName=Mask, texture=/BRCosmetics/Effects/Fort_Effects/Emotes/BentBaton/T_Emote_BentBaton_Trail}.RGBA) {group=Mask, parameterName=Mask Channel Select, sortPriority=33}
  n7 = ScalarParameter() {group=Mask, parameterName=Mask Contrast, sortPriority=34}
  n8 = DynamicParameter() {defaultValue=<rgba 1 1 0 1>, paramNames ["Outline Lerp (Enable in Mat first)", Mask Lerp, Alpha Erosion, Alpha Mask Lerp]}
  n9 = LinearInterpolate(
   A=ScalarParameter() {defaultValue=0.1, group=Color, parameterName=Black Point, sortPriority=34}
   B=NamedRerouteDeclaration(None=MaterialFunctionCall(In=@n6.R, Contrast=@n7) {materialFunction=&CheapContrast}.Result) {name=Mask, nodeColor=<rgba 1 0 0.6094 1>}
   Alpha=@n8.Mask Lerp
  )
  n10 = NamedRerouteUsage()
  n11 = If(
   A=ScalarParameter() {defaultValue=1, group=Fresnel, parameterName=Use Fresnel}
   "A > B"=LinearInterpolate(A=@n9, B=@n10, Alpha=@n10) {constA=0.1}
   "A < B"=@n9
  ) {constB=0.5}
  n12 = Multiply(
   A=@n11
   B=Power(Base=NamedRerouteUsage(), Exp=ScalarParameter() {defaultValue=1, group=Mask, parameterName=Alph To Mask Power, sortPriority=36})
  )
  n13 = ScalarParameter() {group=Mask, parameterName=Apply Alpha To Color MAsk, sortPriority=35}
  n14 = Saturate(None=Max(
    A=LinearInterpolate(A=@n11, B=@n12, Alpha=@n13)
    B=ScalarParameter() {defaultValue=0.1, group=Color, parameterName=Black Point, sortPriority=34}
   ) {constB=0.1})
  n15 = Max(A=ScalarParameter() {defaultValue=8, group=Color, parameterName=Steps, sortPriority=33})
  n16 = Divide(A=Round(None=Multiply(A=@n14, B=@n15)), B=@n15)
  n17 = AppendVector(A=@n16, B=@n16)
  n18 = TextureObjectParameter() {group=Color, parameterName=Gradient Texture, texture=/Game/Effects/Materials/SwingTrails/Textures/T_FX_ColorRamp_Linear}
  n19 = Multiply(
   A=If(
    A=ScalarParameter() {defaultValue=1, group=Color, parameterName=Use Gradient Color Remap, sortPriority=31}
    "A > B"=TextureSample(UVs=@n17, Tex=@n18) {samplerSource=SSM_Clamp_WorldGroupSettings}.RGB
    "A < B"=@n16
   ) {constB=0.5}
   B=ParticleColor().RGB
  )
  n20 = MaterialFunctionCall(Hue Shift Percentage=ScalarParameter() {group=Color, parameterName=Hue Shift, sortPriority=37}, Texture=@n19) {materialFunction=&HueShift}
  n21 = ScalarParameter() {group=Outlines, parameterName=Use Outline Dyn Param Lerp, sortPriority=34}
  n22 = VectorParameter() {
   group=Outlines
   parameterName=Outline Settings
   sortPriority=33
   channelNames {a=Mask Outline Alpha, b=Mask Outline, g=Fresnel Outline Alpha, r=Fresnel Outline}
   defaultValue=<rgba 1 1 1 1>
  }
  n23 = Max(A=Multiply(A=NamedRerouteUsage(), B=@n22.Fresnel Outline), B=Multiply(A=@n22.Mask Outline, B=NamedRerouteUsage()))
  n24 = DynamicParameter() {defaultValue=<rgba 1 1 0 1>, paramNames ["Outline Lerp (Enable in Mat first)", Mask Lerp, Alpha Erosion, Alpha Mask Lerp]}
  n25 = LinearInterpolate(B=@n23, Alpha=@n24."Outline Lerp (Enable in Mat first)")
  MaterialFunctionCall(Emissive=LinearInterpolate(
    A=If(A=@n2, "A > B"=@n20.Result, "A < B"=@n19) {constB=0.5}
    B=VectorParameter() {group=Outlines, parameterName=Outline Color}.RGB
    Alpha=If(A=@n21, "A > B"=@n25, "A < B"=@n23) {constB=0.5}
   ) {constAlpha=1}) {materialFunction=&MF_FX_EmissiveAdjustments}
  n26 = CameraVectorWS()
  n27 = PixelNormalWS()
  n28 = Abs(None=OneMinus(None=DotProduct(A=@n26, B=@n27)))
  n29 = Saturate(None=Power(Base=@n28, Exp=ScalarParameter() {defaultValue=0.5, group=Fresnel, parameterName=Fresnel Pow, sortPriority=34}))
  n30 = ScalarParameter() {group=Fresnel, parameterName=Fresnel Invert, sortPriority=33}
  NamedRerouteDeclaration(None=LinearInterpolate(A=@n29, B=OneMinus(None=@n29), Alpha=Saturate(None=@n30))) {name=Fresnel, nodeColor=<rgba 0.0781 1 0 1>}
  n31 = MaterialFunctionCall(
   In=Add(A=@n28, B=ScalarParameter() {defaultValue=0.1, group=Outlines, parameterName=Outline From Fresnel Thickness, sortPriority=36})
   Contrast=ScalarParameter() {defaultValue=0.5, group=Outlines, parameterName=Outline From Fresnel Contrast, sortPriority=35}
  ) {materialFunction=&CheapContrast}
  NamedRerouteDeclaration(None=Saturate(None=Floor(None=@n31.Result))) {name=Outline, nodeColor=<rgba 1 0 0.3984 1>}
  n32 = DynamicParameter() {
   parameterIndex=2
   defaultValue=<rgba 0 0 1 1>
   paramNames [UV Distortion Offset U, UV Distortion Offset V, UV Distortion Tile U, UV Distortion Tile V]
  }
  n33 = Add(A=ComponentMask(None=@n32.RGBA) {g=true, r=true}, B=TextureCoordinate())
  n34 = VectorParameter() {
   group=UVs
   parameterName=UV Distortion
   sortPriority=38
   channelNames {a=V Tile, b=U Tile, g=V Speed, r=U Speed}
   defaultValue=<rgba 0 0.1 1 1>
  }
  n35 = DynamicParameter() {
   parameterIndex=2
   defaultValue=<rgba 0 0 1 1>
   paramNames [UV Distortion Offset U, UV Distortion Offset V, UV Distortion Tile U, UV Distortion Tile V]
  }
  n36 = LinearInterpolate(
   A=ComponentMask(None=@n34.RGBA) {a=true, b=true}
   B=ComponentMask(None=@n35.RGBA) {a=true, b=true}
   Alpha=ScalarParameter() {group=UVs, parameterName=Use UV Distortion Dyn Param, sortPriority=37}
  )
  n37 = TextureSampleParameter2D(UVs=Panner(Coordinate=Multiply(A=@n33, B=@n36), Speed=ComponentMask(None=@n34.RGBA) {g=true, r=true})) {
   group=UVs
   parameterName=UV Distortion
   samplerType=SAMPLERTYPE_Normal
   sortPriority=36
   texture=/BRCosmetics/Effects/Fort_Effects/Emotes/TremorMark/Materials/T_DecalNoise02
  }
  NamedRerouteDeclaration(None=Subtract(A=ComponentMask(None=@n37.RGB) {g=true, r=true}) {constB=0.5}) {name=UV Distortion, nodeColor=<rgba 0 1 0.6719 1>}
  n38 = VectorParameter() {
   group=UVs
   parameterName=UV Clamp Settings
   sortPriority=35
   channelNames {a=Clamp Mask 2 V, b=Clamp Mask 2 U, g=Clamp Mask 1 V, r=Clamp Mask 1 U}
  }
  n39 = DynamicParameter() {parameterIndex=1, defaultValue=<rgba 0 0 0 0>, paramNames [Mask 1 U Offset, Mask 1 V Offset, Mask 2 U Offset, Mark 2 V Offset]}
  n40 = ComponentMask(None=@n39.RGBA) {a=true, b=true}
  n41 = Add(
   A=TextureCoordinate()
   B=Multiply(A=NamedRerouteUsage(), B=ScalarParameter() {defaultValue=0.05, group=UVs, parameterName=UV Distortion Strength 2, sortPriority=34})
  )
  n42 = VectorParameter() {group=UVs, parameterName=Mask UVs 2, sortPriority=34, channelNames {a=V Tile, b=U Tile, g=V Speed, r=U Speed}, defaultValue=<rgba 0.1 1 1 1>}
  n43 = Panner(
   Coordinate=Multiply(A=Add(A=@n40, B=@n41), B=ComponentMask(None=@n42.RGBA) {a=true, b=true})
   Speed=ComponentMask(None=@n42.RGBA) {g=true, r=true}
  )
  n44 = ComponentMask(None=@n43) {r=true}
  n45 = Clamp(None=@n44) {maxDefault=0.99, minDefault=0.01}
  n46 = ComponentMask(None=@n43) {g=true}
  n47 = Clamp(None=@n46) {maxDefault=0.99, minDefault=0.01}
  NamedRerouteDeclaration(None=AppendVector(A=If(A=@n38.Clamp Mask 2 U, "A > B"=@n45, "A < B"=@n44) {constB=0.5}, B=If(A=@n38.Clamp Mask 2 V, "A > B"=@n47, "A < B"=@n46) {constB=0.5})) {name=Mask UV 2, nodeColor=<rgba 1 0.5391 0 1>}
  n48 = VectorParameter() {
   group=UVs
   parameterName=UV Clamp Settings
   sortPriority=35
   channelNames {a=Clamp Mask 2 V, b=Clamp Mask 2 U, g=Clamp Mask 1 V, r=Clamp Mask 1 U}
  }
  n49 = Add(
   A=TextureCoordinate()
   B=Multiply(A=NamedRerouteUsage(), B=ScalarParameter() {defaultValue=0.1, group=UVs, parameterName=UV Distortion Strength 1, sortPriority=33})
  )
  n50 = ComponentMask(None=@n39.RGBA) {g=true, r=true}
  n51 = VectorParameter() {group=UVs, parameterName=Mask UVs 1, channelNames {a=V Tile, b=U Tile, g=V Speed, r=U Speed}, defaultValue=<rgba 1 0 1 1>}
  n52 = Panner(
   Coordinate=Multiply(A=Add(A=@n49, B=@n50), B=ComponentMask(None=@n51.RGBA) {a=true, b=true})
   Speed=ComponentMask(None=@n51.RGBA) {g=true, r=true}
  )
  n53 = ComponentMask(None=@n52) {r=true}
  n54 = Clamp(None=@n53) {maxDefault=0.99, minDefault=0.01}
  n55 = ComponentMask(None=@n52) {g=true}
  n56 = Clamp(None=@n55) {maxDefault=0.99, minDefault=0.01}
  NamedRerouteDeclaration(None=AppendVector(A=If(A=@n48.Clamp Mask 1 U, "A > B"=@n54, "A < B"=@n53) {constB=0.5}, B=If(A=@n48.Clamp Mask 1 V, "A > B"=@n56, "A < B"=@n55) {constB=0.5})) {name=Mask UV 1, nodeColor=<rgba 0 1 0.1328 1>}
  n57 = NamedRerouteUsage()
  n58 = Add(A=ScalarParameter() {defaultValue=0.001, group=Alpha, parameterName=Erosion Threshold, sortPriority=34}, B=@n1.Alpha Erosion)
  n59 = Add(A=ScalarParameter() {defaultValue=0.1, group=Outlines, parameterName=Outline From Mask Thickness, sortPriority=37}, B=@n58)
  NamedRerouteDeclaration(None=Multiply(A=Step(Y=@n57, X=@n59) {constX=0.15}, B=Step(Y=@n58, X=@n57) {constX=0, constY=0.001})) {name=Mask Outline, nodeColor=<rgba 0.0078 1 0 1>}
  n60 = ScalarParameter() {group=UVs, parameterName=Alpha UV Select, sortPriority=31}
  n61 = TextureSampleParameter2D(UVs=LinearInterpolate(A=NamedRerouteUsage(), B=NamedRerouteUsage(), Alpha=Saturate(None=@n60))) {group=Mask, parameterName=Mask, texture=&T_Emote_BentBaton_Trail}
  n62 = DynamicParameter() {defaultValue=<rgba 1 1 0 1>, paramNames ["Outline Lerp (Enable in Mat first)", Mask Lerp, Alpha Erosion, Alpha Mask Lerp]}
  n63 = LinearInterpolate(
   B=MaterialFunctionCall(
    In=ChannelMaskParameter(None=@n61.RGBA) {group=Alpha, parameterName=Alpha Channel Select, sortPriority=31}.R
    Contrast=ScalarParameter() {group=Alpha, parameterName=Alpha Contrast, sortPriority=34}
   ) {materialFunction=&CheapContrast}.Result
   Alpha=Multiply(A=@n62.Alpha Mask Lerp, B=ScalarParameter() {defaultValue=1, group=Alpha, parameterName=Use Alpha Mask, sortPriority=31})
  ) {constA=1}
  n64 = LinearInterpolate(
   A=@n63
   B=Multiply(A=@n63, B=NamedRerouteUsage())
   Alpha=ScalarParameter() {group=Alpha, parameterName=Combine Mask In Alpha, sortPriority=39}
  )
  n65 = VectorParameter() {
   group=Outlines
   parameterName=Outline Settings
   sortPriority=33
   channelNames {a=Mask Outline Alpha, b=Mask Outline, g=Fresnel Outline Alpha, r=Fresnel Outline}
   defaultValue=<rgba 1 1 1 1>
  }
  n66 = Multiply(A=NamedRerouteUsage(), B=@n65.Fresnel Outline Alpha)
  n67 = Multiply(A=@n65.Mask Outline Alpha, B=NamedRerouteUsage())
  n68 = Multiply(A=ParticleColor().A, B=Max(A=NamedRerouteDeclaration(None=@n64) {name=Alpha Channel, nodeColor=<rgba 0 0.6016 1 1>}, B=Max(A=@n66, B=@n67)))
  n69 = DynamicParameter() {defaultValue=<rgba 1 1 0 1>, paramNames ["Outline Lerp (Enable in Mat first)", Mask Lerp, Alpha Erosion, Alpha Mask Lerp]}
  NamedRerouteDeclaration(None=Saturate(None=Subtract(A=@n68, B=@n69.Alpha Erosion))) {name=Alpha, nodeColor=<rgba 0.5703 1 0 1>}
  n70 = MaterialFunctionCall("FadeSensitivity (optional)"=ScalarParameter() {defaultValue=8, group=Alpha, parameterName=Angle Fade Sensitivity, sortPriority=37}) {materialFunction=&FadeBasedOnViewAngle}
  NamedRerouteDeclaration(None=If(A=ScalarParameter() {group=Alpha, parameterName=Invert Angle Fade, sortPriority=38}, "A > B"=OneMinus(None=@n70.Result), "A < B"=@n70.Result) {constB=0.5}) {name=Angle Fade, nodeColor=<rgba 1 0 0.1875 1>}
  n71 = ScalarParameter() {defaultValue=0.001, group=Alpha, parameterName=Erosion Threshold, sortPriority=34}
  n72 = NamedRerouteUsage()
  n73 = LinearInterpolate(A=@n72, B=Multiply(A=@n72, B=NamedRerouteUsage()), Alpha=ScalarParameter() {group=Alpha, parameterName=Use Angle Fade, sortPriority=36})
  SmoothStep(Min=@n71, Value=@n73)
  Step(Y=@n71, X=@n73) {constX=0.01, constY=0.01}
 ]
}
